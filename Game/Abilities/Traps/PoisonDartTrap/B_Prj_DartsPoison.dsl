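; Projectile Blueprint for the dart fired by the Poison Dart Trap, handling its tracer, impact effects and lifetime.
; Authority times out the dart after a three-tile flight, while remotes run a looping timer that stretches the tracer from the floor trap's socket.
; Hostile player hits shake the camera and attach impact FX to the nearest bone; non-player hits spawn bounce FX.
; FakeEnd hides and stops the dart, lingering briefly for FX before destroying it.
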(event Custom|Timeout
  (CallFunction|FakeEnd))

(event EventBeginPlay
  (bind _particlesystem (Variables|Default|GetParticleSystem))
  (bind _spawned_socket_name (Variables|Default|GetSpawned_Socket_Name))
  (|CallParentFunction)
  (Variables|Default|SetFX%20Linger%20Time (Math|Random|RandomFloatinRange 1.0 4.0))
  (Networking|SwitchHasAuthority
    (:Authority
      (bind _returnvalue (Utilities|Time|SetTimerbyEvent (delegate Custom|Timeout) (/ (Variables|Default|GetThreeTileDistance) (Components|Movement|GetMaxSpeed (Variables|Movement|GetProjectileMovementComponent)))))
      (Variables|Default|SetLifespan%20Timer _returnvalue))
    (:Remote
      (bind _asbp_poison_dart_trap_floor (Utilities|Casting|CastToBP_PoisonDartTrap_Floor (Actor|GetOwner))
        (:then
          (Variables|Default|SetInitialWorldLocation (Transformation|GetSocketLocation (Variables|/Script/FortniteGame.BuildingSMActor|GetStaticMeshComponent _asbp_poison_dart_trap_floor) _spawned_socket_name))
          (Variables|Default|SetInitialWorldRotation (Transformation|GetSocketRotation (Variables|/Script/FortniteGame.BuildingSMActor|GetStaticMeshComponent _asbp_poison_dart_trap_floor) _spawned_socket_name))
          (bind _returnvalue_2 (Rendering|Material|CreateNamedDynamicMaterialInstance _particlesystem "MeshMID" "/Game/Athena/Items/Traps/PoisonDartTrap/FX/M_DartTracer.M_DartTracer"))
          (Variables|VFX|SetParticleMID _returnvalue_2)
          (Effects|Components|ParticleSystem|SetFloatParameter _particlesystem "Lifetime" (+ (/ (Variables|Default|GetThreeTileDistance) (Components|Movement|GetMaxSpeed (Variables|Movement|GetProjectileMovementComponent))) (Variables|Default|GetFX%20Linger%20Time)))
          (bind _returnvalue_3 (Utilities|Time|SetTimerbyEvent (delegate Custom|Tracer) 0.01 true))
          (Variables|Default|SetTracerTimer _returnvalue_3))
        (:CastFailed
          (Development|PrintString "My Owner should be the Poison Trap"))))))

(event Custom|Tracer
  (bind _distancescalefactor (Variables|VFX|GetDistanceScaleFactor))
  (bind _particlesystem (Variables|Default|GetParticleSystem))
  (bind _particlemid (Variables|VFX|GetParticleMID))
  (bind _distance (Class|MissionBlueprintFunctionLibrary|DistanceBetweenTwoVectors (Variables|Default|GetInitialWorldLocation) (Transformation|GetActorLocation)))
  (Effects|Components|ParticleSystem|SetVectorParameter _particlesystem "TracerScale" (Math|Vector|MakeVector (/ _distance _distancescalefactor) 1.0 1.0))
  (Effects|Components|ParticleSystem|SetFloatParameter _particlesystem "TracerUV" (/ _distance _distancescalefactor))
  (Rendering|Material|SetVectorParameterValue _particlemid "ProjectileLocation" (Math|Conversions|ToLinearColor(Vector) (Transformation|GetActorLocation)) :!class /Script/Engine.MaterialInstanceDynamic)
  (Rendering|Material|SetScalarParameterValue _particlemid "Radius" (/ _distance 2.0) :!class /Script/Engine.MaterialInstanceDynamic))

(event EventDestroyed
  (|CallParentFunction))

(event Custom|Fake%20End
  (bind _tracertimer (Variables|Default|GetTracerTimer))
  (Collision|SetActorEnableCollision false)
  (Development|SetHiddeninGame (Variables|Default|GetStaticMesh) true true)
  (Components|Movement|StopMovementImmediately (Variables|Movement|GetProjectileMovementComponent))
  (if (not (Variables|Default|GetHIT))
    (Effects|SpawnEmitteratLocation "/Game/Athena/Items/Traps/PoisonDartTrap/FX/P_Dart_NOHit.P_Dart_NOHit" (Transformation|GetActorLocation)))
  (if (Utilities|Time|IsTimerActivebyHandle _tracertimer)
    (Utilities|Time|ClearandInvalidateTimerbyHandle _tracertimer)
    (else
      _))
  (Utilities|Time|ClearandInvalidateTimerbyHandle (Variables|Default|GetLifespan%20Timer))
  (Utilities|FlowControl|Delay (Variables|Default|GetFX%20Linger%20Time))
  (Actor|DestroyActor))

(event Custom|Collision%20Off
  (Collision|SetActorEnableCollision false))

(event Collision|EventHit
  (params (MyComp :type Primitive%20Component%20Object%20Reference) (Other :type Actor%20Object%20Reference) (OtherComp :type Primitive%20Component%20Object%20Reference) (bSelfMoved :type Boolean) (HitLocation :type Vector :default "0, 0, 0") (HitNormal :type Vector :default "0, 0, 0") (NormalImpulse :type Vector :default "0, 0, 0") (Hit :type Hit%20Result%20Structure%20%28by%20ref%29))
  (bind (_bblockinghit _binitialoverlap _time _distance _location _impactpoint _normal _impactnormal _physmat _hitactor) (Collision|BreakHitResult Hit))
  (bind _returnvalue (Transformation|GetActorForwardVector))
  (bind _returnvalue_2 (* _returnvalue 4e+01))
  (bind _returnvalue_3 (+ _returnvalue_2 _location))
  (|CallParentFunction MyComp Other OtherComp bSelfMoved HitLocation HitNormal NormalImpulse Hit)
  (Variables|Default|SetHIT _bblockinghit)
  (CallFunction|CollisionOff) ; Hit event only fires on client if it happens too quickly after prj spawns
  (bind _asfort_player_pawn (Utilities|Casting|CastToFortPlayerPawn _hitactor)
    (:then
      (switch Utilities|FlowControl|Switch|SwitchonEFortTeamAffiliation (Team|GetTeamAffiliation self _asfort_player_pawn)
        (:Friendly
          _)
        (:Neutral
          _)
        (:Hostile
          (Game|Feedback|ClientStartCameraShake :self (Fortnite|GetFortPlayerControllerfromActor _asfort_player_pawn) :Shake (Variables|Default|GetDamageCameraShake) :Scale 1.0 :PlaySpace "CameraLocal")
          (bind _returnvalue_4 (Effects|SpawnEmitterAttached :EmitterTemplate "/Game/Athena/Items/Traps/PoisonDartTrap/FX/P_Impact_PoisonDart_Player.P_Impact_PoisonDart_Player" :AttachToComponent (Variables|/Script/Engine.Character|GetMesh _asfort_player_pawn) :AttachPointName (select (Utilities|Name|Equal(Name) (.ReturnValue (Components|SkinnedMesh|FindClosestBone (Variables|/Script/Engine.Character|GetMesh _asfort_player_pawn) _returnvalue_3 0.0 false)) "head") "neck_01" (.ReturnValue (Components|SkinnedMesh|FindClosestBone (Variables|/Script/Engine.Character|GetMesh _asfort_player_pawn) _returnvalue_3 0.0 false))) :Scale "1.000000,1.000000,1.000000" :LocationType "SnapToTarget" :bAutoDestroy true :PoolingMethod "None" :bAutoActivate true))
          (Utilities|IsValid _returnvalue_4
            (:Is%20Valid
              (Transformation|SetWorldRotation _returnvalue_4 (Math|Conversions|RotationFromXVector (* (Transformation|GetActorForwardVector) -1.0)) false false) ; returns none when hits something that's too close
            )
            (:Is%20Not%20Valid
              _))
          (Utilities|FlowControl|Delay 0.05)
          (CallFunction|FakeEnd)))
      (CallFunction|FakeEnd))
    (:CastFailed
      (Effects|SpawnEmitteratLocation "/Game/Athena/Items/Traps/PoisonDartTrap/FX/P_Impact_PoisonDart_Bounce.P_Impact_PoisonDart_Bounce" _returnvalue_3 (Math|Conversions|RotationFromXVector (* _returnvalue -1.0)))
      (CallFunction|FakeEnd))))

(fn TeamCheck
  (bind _returnvalue (Fortnite|GetLocalFortPlayerControllers true))
  (if (and (Utilities|Enum|Equal(Enum) (Fortnite|GetSubGame) "Athena") (> (Utilities|Array|Length _returnvalue) 0))
    (if (Math|Byte|Equal(Byte) 0 (Team|GetActorTeam (Utilities|Array|Get(acopy) _returnvalue)))
      _)))

(fn ConstructionScript)
